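; Looping gameplay cue base for PinkOatmeal "Inside" that plays 2D music for the affected local player.
; On application it caches the target pawn, fades out any earlier music, spawns the music asset and sets its IsSkydiving parameter once.
; A local-effects check rejects character-vehicle targets and accepts locally controlled pawns or those in a locally controlled vehicle.
; On removal it fades the music out.
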(event EventOnApplication(Cascade)
  (params (MyTarget :type Actor%20Object%20Reference) (Parameters :type Gameplay%20Cue%20Parameters%20Structure%20%28by%20ref%29) (ParticleComponents :type Array%20of%20Cascade%20Particle%20System%20Component%20Object%20References) (AudioComponents :type Array%20of%20Audio%20Component%20Object%20References) (BurstCameraShakeInstance :type Camera%20Shake%20Base%20Object%20Reference) (BurstDecalInstance :type Decal%20Actor%20Object%20Reference))
  (bind _asfort_player_pawn (Utilities|Casting|CastToFortPlayerPawn MyTarget))
  (Variables|Default|SetTargetPlayerPawn _asfort_player_pawn)
  (bind _returnvalue (CallFunction|ShouldTriggerLocalEffects))
  (if _returnvalue
    (bind _musicaudio (Variables|Default|GetMusicAudio)
      (:then
        (Audio|Components|Audio|FadeOut _musicaudio 1.0 0.0 "Linear" :!class /Script/Engine.AudioComponent))
      (:else
        _))
    (bind _musicasset (Variables|Audio|GetMusicAsset)
      (:then
        (bind _returnvalue_2 (Audio|SpawnSound2D 0 _musicasset))
        (bind _output_get (Variables|Default|SetMusicAudio _returnvalue_2))
        (Audio|Parameter|SetIntegerParameter _output_get "IsSkydiving" (Math|Conversions|ToInteger(Boolean) (Parachute|IsSkydiving (Variables|Default|GetTargetPlayerPawn))) :!class /Script/AudioExtensions.AudioParameterControllerInterface))
      (:else))))

(event EventOnRemoval(Cascade)
  (params (MyTarget :type Actor%20Object%20Reference) (Parameters :type Gameplay%20Cue%20Parameters%20Structure%20%28by%20ref%29) (ParticleComponents :type Array%20of%20Cascade%20Particle%20System%20Component%20Object%20References) (AudioComponents :type Array%20of%20Audio%20Component%20Object%20References) (BurstCameraShakeInstance :type Camera%20Shake%20Base%20Object%20Reference) (BurstDecalInstance :type Decal%20Actor%20Object%20Reference))
  (Utilities|Casting|CastToFortPlayerPawn MyTarget)
  (bind _musicaudio (Variables|Default|GetMusicAudio)
    (:then
      (Audio|Components|Audio|FadeOut _musicaudio 3.0 0.0 "Linear" :!class /Script/Engine.AudioComponent))
    (:else)))

(fn ShouldTriggerLocalEffects
  (Utilities|Casting|CastToFortCharacterVehicle (Variables|Default|GetTargetPlayerPawn)
    (:then
      (return false))
    (:CastFailed
      (if (Pawn|IsLocallyControlled (Variables|Default|GetTargetPlayerPawn))
        (return true)
        (else
          (bind _aspawn (Utilities|Casting|CastToPawn (Vehicle|GetVehicleActor (Variables|Default|GetTargetPlayerPawn) :!class /Script/FortniteGame.FortPlayerPawn))
            (:then
              (return (Pawn|IsLocallyControlled _aspawn)))
            (:CastFailed
              (return false))))))))

(fn ConstructionScript)
