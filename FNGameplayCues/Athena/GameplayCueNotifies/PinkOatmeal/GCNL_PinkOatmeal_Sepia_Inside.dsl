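; Looping gameplay cue that adds Western-themed footstep sounds to a player pawn while the PinkOatmeal sepia "inside" effect is active, apparently inside a Rift Beacon bubble.
; Application and removal call the parent, then bind or unbind a handler on the target pawn's footstep event.
; Each footstep plays a sound at the pawn's location, choosing a first-person or third-person cue by whether the pawn is locally controlled.
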(event EventOnApplication(Cascade)
  (params (MyTarget :type Actor%20Object%20Reference) (Parameters :type Gameplay%20Cue%20Parameters%20Structure%20%28by%20ref%29) (ParticleComponents :type Array%20of%20Cascade%20Particle%20System%20Component%20Object%20References) (AudioComponents :type Array%20of%20Audio%20Component%20Object%20References) (BurstCameraShakeInstance :type Camera%20Shake%20Base%20Object%20Reference) (BurstDecalInstance :type Decal%20Actor%20Object%20Reference))
  (|CallParentFunction MyTarget Parameters ParticleComponents AudioComponents BurstCameraShakeInstance BurstDecalInstance)
  (Utilities|Casting|CastToFortPlayerPawn MyTarget)
  (bind _targetplayerpawn (Variables|Default|GetTargetPlayerPawn)
    (:then
      (EventDispatchers|BindEventtoOnFootstepEvent _targetplayerpawn (EventDispatchers|CreateEvent :!event Custom|On%20Player%20Footstep)))
    (:else)))

(event EventOnRemoval(Cascade)
  (params (MyTarget :type Actor%20Object%20Reference) (Parameters :type Gameplay%20Cue%20Parameters%20Structure%20%28by%20ref%29) (ParticleComponents :type Array%20of%20Cascade%20Particle%20System%20Component%20Object%20References) (AudioComponents :type Array%20of%20Audio%20Component%20Object%20References) (BurstCameraShakeInstance :type Camera%20Shake%20Base%20Object%20Reference) (BurstDecalInstance :type Decal%20Actor%20Object%20Reference))
  (|CallParentFunction MyTarget Parameters ParticleComponents AudioComponents BurstCameraShakeInstance BurstDecalInstance)
  (Utilities|Casting|CastToFortPlayerPawn MyTarget)
  (bind _targetplayerpawn (Variables|Default|GetTargetPlayerPawn)
    (:then
      (EventDispatchers|UnbindEventfromOnFootstepEvent _targetplayerpawn (EventDispatchers|CreateEvent :!event Custom|On%20Player%20Footstep)))
    (:else)))

(event Custom|On%20Player%20Footstep
  (params (FootIndex :type Integer))
  (bind _targetplayerpawn (Variables|Default|GetTargetPlayerPawn)
    (:then
      (Audio|PlaySoundatLocation _targetplayerpawn (select (Pawn|IsLocallyControlled _targetplayerpawn) "/FNGameplayCues/Athena/Sounds/Events/RiftBeacon/RiftBubble/Western/Western_Inside_Step_Cue.Western_Inside_Step_Cue" "/FNGameplayCues/Athena/Sounds/Events/RiftBeacon/RiftBubble/Western/Western_Inside_Step_3P_Cue.Western_Inside_Step_3P_Cue") (Transformation|GetActorLocation _targetplayerpawn)))
    (:else)))

(fn ConstructionScript)
